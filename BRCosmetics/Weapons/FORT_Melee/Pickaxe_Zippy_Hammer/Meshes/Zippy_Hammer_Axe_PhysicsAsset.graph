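# Physics asset for the Zippy Hammer pickaxe mesh, a melee weapon cosmetic.
# It holds a single body on the b_AR_Root bone, shaped as one capsule offset from the bone and rotated to fit the weapon.
# It has no constraints, so the weapon simulates and collides as one rigid piece.
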PhysicsAsset /BRCosmetics/Weapons/FORT_Melee/Pickaxe_Zippy_Hammer/Meshes/Zippy_Hammer_Axe_PhysicsAsset
bodies {b_AR_Root [{shapeType=Capsule, length=69.1066, radius=20.4873, center=<7.5171 -0.3721 26.2859>, rotation=<rot 0 -90 0.5595>}]}
constraints []
